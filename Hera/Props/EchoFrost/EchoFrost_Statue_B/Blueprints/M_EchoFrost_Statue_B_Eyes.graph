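# Translucent, unlit, two-sided material for the glowing eyes of the EchoFrost Statue B, usable on Niagara mesh particles.
# Vertex red blends front-face pupil and iris sphere masks into wispy, view-angle-faded trails of panning smoke and tear masks, with a Glacier variant.
# The smoke also drives X-axis vertex wobble along V.
# A two-color gradient tinted by particle color feeds HDR emissive, while depth fade softens opacity on higher quality levels.
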Material /Hera/Props/EchoFrost/EchoFrost_Statue_B/Blueprints/M_EchoFrost_Statue_B_Eyes
graph {
 properties {
  bAutomaticallySetUsageInEditor=false
  bEnableResponsiveAA=true
  bUsedWithNiagaraMeshParticles=true
  bUsedWithStaticMesh=false
  blendMode=BLEND_Translucent
  pixelDepthOffsetMode=PDOM_Legacy
  shadingModel=MSM_Unlit
  twoSided=true
  usedShadingModels=Unlit
 }
 nodes [
  n1 = ScalarParameter() {defaultValue=0.5, parameterName=WPOAmount}
  n2 = Multiply(A=Constant3Vector() {constant=<rgba -1 0 0 0>}, B=@n1)
  n3 = Multiply(A=Constant3Vector() {constant=<rgba 1 0 0 0>}, B=@n1)
  n4 = VertexColor()
  n5 = Add(A=TextureCoordinate(), B=AppendVector(A=Multiply(A=@n4.A) {constB=0.2}, B=Multiply(A=@n4.A) {constB=0.5}))
  n6 = Multiply(
   A=@n5
   B=AppendVector(A=ScalarParameter() {defaultValue=1, parameterName=U Scale}, B=ScalarParameter() {defaultValue=0.25, parameterName=V Scale})
  )
  n7 = Multiply(
   A=@n5
   B=AppendVector(A=ScalarParameter() {defaultValue=1.5, parameterName=U Scale_B}, B=ScalarParameter() {defaultValue=1, parameterName=V Scale_B})
  )
  n8 = Multiply(
   A=TextureSample(UVs=Panner(Coordinate=@n6) {speedX=0.13, speedY=0.125}) {samplerSource=SSM_Wrap_WorldGroupSettings, texture=&T_Inky_Smoke_Tile}.R
   B=TextureSample(UVs=Panner(Coordinate=@n7) {speedX=-0.18, speedY=-0.125}) {samplerSource=SSM_Wrap_WorldGroupSettings, texture=&T_Inky_Smoke_Tile}.R
  )
  n9 = VectorParameter() {parameterName=Tear Mask Controls, channelNames {a=Intensity, b=V Offset, g=Shape Bottom, r=Shape Top}, defaultValue=<rgba 0 100 0 0.25>}
  n10 = Add(A=TextureCoordinate(), B=AppendVector(A=Constant(), B=@n9.V Offset))
  n11 = ComponentMask(None=@n10) {r=true}
  n12 = ComponentMask(None=@n10) {g=true}
  n13 = Multiply(A=@n12, B=@n12)
  n14 = Multiply(
   A=Power(Base=Sine(None=@n11) {period=2}, Exp=LinearInterpolate(A=@n9.Shape Top, B=@n9.Shape Top, Alpha=@n13) {constA=10, constB=80}) {constExponent=100}
   B=@n9.Intensity
  ) {constB=2}
  n15 = Add(A=Add(A=@n8, B=@n14)) {constB=0.1}
  n16 = TextureCoordinate()
  n17 = VertexColor()
  LinearInterpolate(B=LinearInterpolate(B=LinearInterpolate(A=@n2, B=@n3, Alpha=@n15), Alpha=ComponentMask(None=@n16) {g=true}), Alpha=@n17.R)
  n18 = VectorParameter() {group=Color, parameterName=EyeColor, sortPriority=3, defaultValue=<rgba 0.136 0.3224 2 1>}
  n19 = ComponentMask(None=TextureCoordinate()) {g=true}
  n20 = MaterialFunctionCall() {materialFunction=&BoundingBoxBased_0-1_UVW}
  n21 = ScalarParameter() {defaultValue=1, parameterName=UseBBoxGradient}
  n22 = LinearInterpolate(
   A=@n18.RGB
   B=VectorParameter() {group=Color, parameterName=Color_2, sortPriority=5, defaultValue=<rgba 0.136 0.3224 2 1>}.RGB
   Alpha=SmoothStep(
    Min=ScalarParameter() {defaultValue=0.35, parameterName=ColorGradientSmoothStepMin}
    Max=ScalarParameter() {defaultValue=0.65, parameterName=ColorGradientSmoothStepMax}
    Value=LinearInterpolate(A=@n19, B=@n20.B, Alpha=@n21)
   ) {constMax=0.65, constMin=0.35}
  )
  n23 = VertexColor()
  n24 = Multiply(
   A=LinearInterpolate(
    A=@n18.RGB
    B=LinearInterpolate(A=@n18.RGB, B=@n22, Alpha=@n23.R)
    Alpha=ScalarParameter() {defaultValue=1, group=Color, parameterName=Use Color Gradient, sortPriority=1}
   )
   B=ParticleColor().RGB
  )
  n25 = Clamp(None=@n24)
  n26 = ScalarParameter() {defaultValue=2, parameterName=HDR_Intensity}
  MaterialFunctionCall(Emissive=Multiply(
    A=MaterialFunctionCall(LDR Input=@n25, HDR Tint=@n24, HDR Intensity Multiplier=@n26) {materialFunction=&DeriveHDRfromLDR}.RemappedRange
    B=ScalarParameter() {defaultValue=0.25, parameterName=Overall_Intensity}
   )) {materialFunction=&MF_FX_EmissiveAdjustments}
  n27 = ComponentMask(None=VectorParameter() {parameterName=PupilOffset, defaultValue=<rgba 0 -0.15 0 0>}.RGB) {g=true, r=true}
  n28 = ComponentMask(None=TextureCoordinate()) {r=true}
  n29 = SphereMask(
   A=Add(A=TextureCoordinate() {uTiling=0.5}, B=LinearInterpolate(B=@n27, Alpha=@n28))
   B=ComponentMask(None=VectorParameter() {parameterName=PupilCenter, defaultValue=<rgba 0.25 0.5 0 0>}.RGB) {g=true, r=true}
   Radius=ScalarParameter() {defaultValue=0.2, parameterName=Pupil Radius}
   Hardness=ScalarParameter() {parameterName=Pupil Hardness}
  )
  n30 = TextureCoordinate()
  n31 = ComponentMask(None=@n30) {g=true}
  n32 = LinearInterpolate(Alpha=ComponentMask(None=@n30) {r=true}) {constB=3}
  n33 = VectorParameter() {parameterName=IrisCenter, defaultValue=<rgba 0.5 0.5 0 0>}
  n34 = Multiply(A=SphereMask(
    A=TextureCoordinate()
    B=ComponentMask(None=@n33.RGB) {g=true, r=true}
    Radius=ScalarParameter() {defaultValue=0.4, parameterName=Iris Radius}
    Hardness=ScalarParameter() {parameterName=Iris Hardness}
   )) {constB=2}
  n35 = Add(
   A=Multiply(A=Power(Base=@n29) {constExponent=8}, B=ScalarParameter() {defaultValue=1.8})
   B=Multiply(A=Power(Base=@n31, Exp=@n32), B=Add(A=@n29, B=@n34))
  )
  n36 = TwoSidedSign()
  n37 = TextureCoordinate()
  n38 = TextureCoordinate()
  n39 = TextureCoordinate()
  n40 = Multiply(A=OneMinus(None=ComponentMask(None=@n38) {g=true}), B=Power(Base=ComponentMask(None=@n39) {g=true}) {constExponent=0.8})
  n41 = ScalarParameter() {defaultValue=10, parameterName=TopContrastIntensitry}
  n42 = Multiply(A=Sine(None=ComponentMask(None=@n37) {r=true}) {period=2}, B=Clamp(None=Multiply(A=@n40, B=@n41) {constB=10}))
  n43 = Multiply(A=TextureCoordinate(), B=AppendVector(A=Constant() {r=1}, B=ScalarParameter() {defaultValue=0.7, parameterName=UV Clamp Slot}))
  n44 = Panner(Coordinate=@n43) {speedX=0.05}
  n45 = Panner(Coordinate=@n43) {speedX=-0.0336}
  n46 = OneMinus(None=Multiply(
    A=TextureSample(UVs=@n44) {samplerType=SAMPLERTYPE_LinearColor, texture=/Game/Effects/Fort_Effects/Textures/Tile/T_Raygun_Muzzle_01}.R
    B=TextureSample(UVs=@n45) {samplerType=SAMPLERTYPE_LinearColor, texture=&T_Raygun_Muzzle_01}.R
   ))
  n47 = LinearInterpolate(A=Multiply(A=Multiply(A=@n42, B=@n46), B=@n15), B=NamedRerouteUsage(), Alpha=ScalarParameter() {parameterName=Glacier Variant})
  n48 = MaterialFunctionCall("FadeSensitivity (optional)"=ScalarParameter() {defaultValue=3, parameterName=FadeSensitivity}) {materialFunction=&FadeBasedOnViewAngle}
  n49 = Multiply(A=ParticleColor().A, B=LinearInterpolate(A=LinearInterpolate(B=@n35, Alpha=@n36), B=Multiply(A=@n47, B=@n48.Result), Alpha=@n17.R))
  n50 = DepthFade(Opacity=@n49, FadeDistance=ScalarParameter() {defaultValue=15, parameterName=Depth Fade Distance})
  Multiply(A=ScalarParameter() {defaultValue=1, parameterName=Alpha}, B=QualitySwitch(Default=@n50, Low=@n49, High=@n50, Medium=@n49, Epic=@n50))
  n51 = LinearInterpolate(Alpha=@n8) {constA=0.1}
  n52 = Power(Base=OneMinus(None=@n12))
  NamedRerouteDeclaration(None=Multiply(A=Multiply(A=@n51) {constB=1.2}, B=Multiply(A=@n14, B=@n52))) {name=Glacier Variant Mask, nodeColor=<rgba 0.2891 1 0 1>}
 ]
}
